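# Niagara cosmetic effect for the Weeping Woods Winter backpack, with mesh bees and drool attached to sockets on the backpack skeletal mesh.
# Two bee emitters spawn at one socket and stop spawning while skydiving: one swarms back via curl noise and point attraction, the other drifts away on velocity and rising noise.
# Drool bursts swing as pendulums from random sockets, stretch and drop, feeding pivot and tip positions to dynamic material parameters.
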NiagaraSystem /BRCosmetics/Accessories/FORT_Backpacks/Backpack_M_MED_Weeping_Woods_Winter/NS_WeepingWoods_Winter_Backpack
dependencies {
 usedDataInterfaces [&NiagaraDataInterfaceCurve]
 usedDynamicInputs []
 usedModules [
  &SystemState
  &self:SystemSpawnScript.SystemScriptSource.SystemScriptGraph.NiagaraNodeAssignment_0.SetVariables_#1
 ]
}
emitters {
 Bees {
  emitterUpdateScript [
   &EmitterState()
   &SpawnRate(
    SpawnRate=&Lerp_Float(A=1, B=0, Alpha=&MakeFloatFromInt(Int=$User.IsSkydiving))
    Use Spawn Probability=true
   )
  ]
  particleSpawnScript [
   /Niagara/Modules/Spawn/Initialization/InitializeParticle(
    Lifetime=&Lerp_Float(A=2.5, B=4, Alpha=&Power(a=&RandomRangeFloat(Minimum=0, Maximum=1), b=4))
    Mesh Scale=&Multiply_VectorByFloat(Vector=<1 1 1>, Float=&UniformRangedFloat(Minimum=0.7, Maximum=0.8))
    Module.Write Color=true
    Module.Write Lifetime=true
    Module.Write Mass=false
    Module.Write Material Random=true
    Module.Write Position=false
    Module.Write Scale=true
    Module.Write SpriteRotation=false
    Module.Write SpriteSize=false
    Position=&SimulationPosition()
   )
   &SkeletalMeshLocation(
    Bone / Socket Index=3
    Bone Sampling Mode=::NewEnumerator4
    Inherit Velocity=false
    MaskSpawnGroup=false
    Sampled Position Offset=&RandomRangeVector(Minimum=<-3 -1 2>, Maximum=<3 -2 3>)
    Sampled Position Offset Coordinate Space:ENiagaraMeshTransforms=::NewEnumerator3
    Skeletal Mesh=$User.SkeletalMesh
   )
   &AddVelocity(Velocity=&Multiply_Vector(A=$Output.SkeletalMeshLocation.SampledPosition, B=&RandomRangeVector(Minimum=<0 3 0>, Maximum=<0 8 0>))) {enabled=false}
   &InitialMeshOrientation(
    Orientation Axis=<0 0 1>
    Orientation Vector=<0 0 1>
    Rotation=&RandomRangeVector(Minimum=<0 0 -0.2>, Maximum=<0 0 0.2>)
    Use Orientation Vector=true
    Use Rotation Vector=true
   )
  ]
  particleUpdateScript [
   &ParticleState()
   &SkeletalMeshLocation(
    Apply To Particle Position=false
    Bone / Socket Index=3
    Bone Sampling Mode=::NewEnumerator4
    Inherit Velocity=false
    MaskSpawnGroup=false
    Sampled Position Offset=&Add_Vector(
     A=<-22.1231 -11 -0.8408>
     B=&RotateVectorByQuaternion(Vector To Rotate=&RandomRangeVector(Minimum=<0 16.1011 0>, Maximum=<30 0 2>), Quaternion=<0 -0.15 0 1>)
    )
    Sampled Position Offset Coordinate Space:ENiagaraMeshTransforms=::NewEnumerator3
    Skeletal Mesh=$User.SkeletalMesh
   ) {module=SkeletalMeshLocation001}
   /Niagara/Modules/Update/Forces/CurlNoiseForce(
    Noise Frequency=0.5
    Noise Strength=&FloatFromCurve(
     FloatCurve=DataInterface() {propertyValues {Curve {keys [
         {interpMode=RCIM_Cubic, tangentMode=RCTM_User, time=0, value=0}
         {arriveTangent=0.0329, interpMode=RCIM_Cubic, leaveTangent=-2.5509, tangentMode=RCTM_Break, time=0.1, value=1}
         <at 1 0>
        ]}}}
     Scale Curve=1000
    )
    Pan Noise=true
   )
   /Niagara/Modules/Update/Forces/PointAttractionForce(
    AttractionStrength:NiagaraFloat=&Lerp_Float(A=0, B=100, Alpha=$Particles.NormalizedAge)
    AttractorPosition:Vector3f=&Lerp_Vector(
     A=&Add_Vector(
      A=$Output.SkeletalMeshLocation001.SampledPosition
      B=&RandomRangeVector(Minimum=<-7 -20 3>, Maximum=<7 -40 5>)
     )
     B=&Add_Vector(
      A=$Output.SkeletalMeshLocation001.SampledPosition
      B=&RandomRangeVector(Minimum=<0 0 2>, Maximum=<0 1 4>)
     )
     Alpha=&FloatFromCurve(FloatCurve=DataInterface() {propertyValues {Curve {keys [{arriveTangent=1.9279, interpMode=RCIM_Cubic, leaveTangent=1.9279, tangentMode=RCTM_User, time=0, value=0}, <at 1 1>]}}})
    )
    Falloff Exponent=1
    Module.Kill Within Radius=false
    Module.Use Falloff=true
   ) {module=PointAttractionForce002}
   &Drag(Drag=5, Use Linear Drag=true, Use Rotational Drag=false)
   &SolveForcesAndVelocity(Clamp Velocity=false, Limit Acceleration=false)
   &ScaleMeshSize(Scale Factor=&VectorFromFloat(Value=&FloatFromCurve(FloatCurve=DataInterface() {propertyValues {Curve {keys [
          {interpMode=RCIM_Cubic, tangentMode=RCTM_User, time=-0.0046, value=1.0113}
          {arriveTangent=-1.0067, interpMode=RCIM_Cubic, leaveTangent=-1.0067, tangentMode=RCTM_User, time=0.9283, value=0.7392}
          <at 1 0>
         ]}}})))
   &UpdateMeshOrientation(Rotation Rate=&RandomRangeFloat(Minimum=-0.6, Maximum=0.6), Rotation Vector=<0 0 1>)
   /Game/Effects/Niagara/Emitters/Water/FN_DistanceKillParticles(AgeReduiceAmount=0.1, DistanceFalloffExp=0.75)
  ]
  rendererClasses [&NiagaraMeshRendererProperties]
  renderers [{rendererClass=&NiagaraMeshRendererProperties, rendererIndex=0}]
 }
 Bees_fly_away {
  emitterUpdateScript [
   &EmitterState()
   &SpawnRate(
    SpawnRate=&Lerp_Float(A=1, B=0, Alpha=&MakeFloatFromInt(Int=$User.IsSkydiving))
    Use Spawn Probability=true
   )
  ]
  particleSpawnScript [
   &InitializeParticle(
    Lifetime=&Lerp_Float(A=2.5, B=4, Alpha=&Power(a=&RandomRangeFloat(Minimum=0, Maximum=1), b=4))
    Mesh Scale=&Multiply_VectorByFloat(Vector=<1 1 1>, Float=&UniformRangedFloat(Minimum=0.7, Maximum=0.8))
    Module.Write Color=true
    Module.Write Lifetime=true
    Module.Write Mass=false
    Module.Write Material Random=true
    Module.Write Position=false
    Module.Write Scale=true
    Module.Write SpriteRotation=false
    Module.Write SpriteSize=false
    Position=&SimulationPosition()
   )
   &SkeletalMeshLocation(
    Bone / Socket Index=3
    Bone Sampling Mode=::NewEnumerator4
    Inherit Velocity=false
    MaskSpawnGroup=false
    Sampled Position Offset=&RandomRangeVector(Minimum=<-3 -1 2>, Maximum=<3 -2 3>)
    Sampled Position Offset Coordinate Space:ENiagaraMeshTransforms=::NewEnumerator3
    Skeletal Mesh=$User.SkeletalMesh
   )
   &AddVelocity(Velocity=&Multiply_Vector(
     A=$Output.SkeletalMeshLocation.SampledPosition
     B=&RandomRangeVector(Minimum=<-3 5 0>, Maximum=<3 8 3>)
    ))
   &InitialMeshOrientation(
    Orientation Axis=<0 0 1>
    Orientation Vector=<0 0 1>
    Rotation=&RandomRangeVector(Minimum=<0 0 -0.2>, Maximum=<0 0 0.2>)
    Use Orientation Vector=true
    Use Rotation Vector=true
   )
  ]
  particleUpdateScript [
   &ParticleState()
   &SkeletalMeshLocation(
    Apply To Particle Position=false
    Bone / Socket Index=3
    Bone Sampling Mode=::NewEnumerator4
    Inherit Velocity=false
    MaskSpawnGroup=false
    Sampled Position Offset=&Add_Vector(
     A=<-22.1231 -11 -0.8408>
     B=&RotateVectorByQuaternion(Vector To Rotate=&RandomRangeVector(Minimum=<0 16.1011 0>, Maximum=<30 0 2>), Quaternion=<0 -0.15 0 1>)
    )
    Sampled Position Offset Coordinate Space:ENiagaraMeshTransforms=::NewEnumerator3
    Skeletal Mesh=$User.SkeletalMesh
   ) {module=SkeletalMeshLocation001}
   &CurlNoiseForce(
    Noise Frequency=0.5
    Noise Strength=&FloatFromCurve(
     FloatCurve=DataInterface() {propertyValues {Curve {keys [
         {arriveTangent=2.1993, interpMode=RCIM_Cubic, leaveTangent=2.1993, tangentMode=RCTM_User, time=0, value=0}
         {arriveTangent=0.0329, interpMode=RCIM_Cubic, leaveTangent=0.1296, tangentMode=RCTM_Break, time=0.2, value=0.2}
         <at 1 1>
        ]}}}
     Scale Curve=5000
    )
    Pan Noise=true
    Pan Noise Field=<0 0 2>
   )
   &PointAttractionForce(
    AttractionStrength:NiagaraFloat=&Lerp_Float(A=0, B=100, Alpha=$Particles.NormalizedAge)
    AttractorPosition:Vector3f=&Lerp_Vector(
     A=&Add_Vector(
      A=$Output.SkeletalMeshLocation001.SampledPosition
      B=&RandomRangeVector(Minimum=<-10 -5 2>, Maximum=<10 -15 3>)
     )
     B=&Add_Vector(
      A=$Output.SkeletalMeshLocation001.SampledPosition
      B=&RandomRangeVector(Minimum=<0 1 2>, Maximum=<0 2 2>)
     )
     Alpha=&FloatFromCurve(FloatCurve=DataInterface() {propertyValues {Curve {keys [{arriveTangent=0.1313, interpMode=RCIM_Cubic, leaveTangent=0.1313, tangentMode=RCTM_User, time=0, value=0}, <at 1 1>]}}})
    )
    Falloff Exponent=1
    Module.Kill Within Radius=false
    Module.Use Falloff=true
   ) {enabled=false, module=PointAttractionForce002}
   &Drag(Drag=5, Use Linear Drag=true, Use Rotational Drag=false)
   &SolveForcesAndVelocity(Clamp Velocity=false, Limit Acceleration=false)
   &ScaleMeshSize(Scale Factor=&VectorFromFloat(Value=&FloatFromCurve(FloatCurve=DataInterface() {propertyValues {Curve {keys [
          {interpMode=RCIM_Cubic, tangentMode=RCTM_User, time=-0.0046, value=1.0113}
          {arriveTangent=-1.0067, interpMode=RCIM_Cubic, leaveTangent=-1.0067, tangentMode=RCTM_User, time=0.9283, value=0.7392}
          <at 1 0>
         ]}}})))
   &UpdateMeshOrientation(Rotation Rate=&RandomRangeFloat(Minimum=-0.6, Maximum=0.6), Rotation Vector=<0 0 1>)
   &FN_DistanceKillParticles(AgeReduiceAmount=0.1, DistanceFalloffExp=0.75)
  ]
  rendererClasses [&NiagaraMeshRendererProperties]
  renderers [{rendererClass=&NiagaraMeshRendererProperties, rendererIndex=0}]
 }
 Drool {
  emitterUpdateScript [
   &EmitterState(
    Delay First Loop Only=true
    Inactive Response=::NewEnumerator1
    Life Cycle Mode=::NewEnumerator1
    Loop Delay=&FN_FrontEndFloatModifier(Gameplay=0, Front End=1)
    Loop Duration=3.5
    UseLoopDelay=true
   )
   &SpawnBurst_Instantaneous(Spawn Count=1, Use Spawn Probability=false)
  ]
  particleSpawnScript [
   /SetVariables(
    Particles.PendulumDrag=5
    Particles.PendulumLengthScale=0.125
    Particles.PendulumMaskFalloff=1.5
    Particles.PendulumRigidity=0.2
    Particles.RandBone:NiagaraInt32=&RandomRangeInt(Minimum=0, Maximum=2)
   ) {isSetParameters=true}
   &InitializeParticle(
    Lifetime=&UniformRangedFloat(Minimum=2, Maximum=3.5)
    Mesh Scale=&Multiply_Vector(A=<1 1 1>, B=&MakeVector(X=0.12, Y=0.12, Z=$Particles.PendulumLengthScale))
    Module.Write Color=true
    Module.Write Lifetime=true
    Module.Write Mass=true
    Module.Write Material Random=true
    Module.Write Position=false
    Module.Write Scale=true
    Module.Write SpriteRotation=false
    Module.Write SpriteSize=true
    Sprite Size=<1 1>
   )
   &SkeletalMeshLocation(
    Bone / Socket Index=$Particles.RandBone
    Bone Sampling Mode=::NewEnumerator4
    Inherit Velocity=false
    MaskSpawnGroup=false
    Skeletal Mesh=$User.SkeletalMesh
   ) {module=SkeletalMeshLocation001}
  ]
  particleUpdateScript [
   &SkeletalMeshLocation(
    Apply To Particle Position=false
    Bone / Socket Index=$Particles.RandBone
    Bone Sampling Mode=::NewEnumerator4
    Inherit Velocity=false
    MaskSpawnGroup=false
    Skeletal Mesh=$User.SkeletalMesh
   )
   &ParticleState()
   /SetVariables(Particles.MeshOrigin:Vector3f=&Add_Vector(
     A=$Output.SkeletalMeshLocation.SampledPosition
     B=&MakeVector(
      X=0
      Y=0
      Z=&FloatFromCurve(
       FloatCurve=DataInterface() {propertyValues {Curve {keys [
           <at 0 0>
           {arriveTangent=1.5112, interpMode=RCIM_Cubic, leaveTangent=1.5112, tangentMode=RCTM_User, time=0.7, value=0}
           {arriveTangent=4.7214, interpMode=RCIM_Cubic, leaveTangent=4.7214, tangentMode=RCTM_User, time=1, value=1}
          ]}}}
       Scale Curve=-170
      )
     )
    )) {isSetParameters=true}
   &Drag(Drag=$Particles.PendulumDrag, Use Linear Drag=true, Use Rotational Drag=false)
   &GravityForce(Gravity=<0 0 -900>)
   &PendulumSetup(
    Module.Enable Spring Driver=false
    Pendulum Length:NiagaraFloat=&Multiply_Float(A=100, B=$Particles.PendulumLengthScale)
    Pendulum Pivot:Vector3f=$Particles.MeshOrigin
    "Spring Driven Constraint (Tightness)"=3
   )
   &SolveForcesAndVelocity(Clamp Velocity=false, Limit Acceleration=false)
   &PendulumConstraint()
   &CalculateAccurateVelocity()
   /SetVariables(Particles.MeshFacing:Vector3f=&NormalizeVector(Vector To Normalize=&Subtract_Vector(A=$Particles.MeshOrigin, B=Linked() {linkedVariable {name=Particles.Position, type=NiagaraPosition}}))) {isSetParameters=true}
   /Niagara/Modules/Update/Orientation/V2/OrientMeshToVector(
    Look At Direction:Vector3f=&Multiply_VectorByFloat(Vector=$Particles.MeshFacing, Float=1)
    Override Quat=false
    Reference Axis Vector:Vector3f=<0 0 1>
    UseOptionalSideAxis=false
    UseOptionalUpAxis=false
   )
   &DynamicMaterialParameters(
    Index 2 Param 1=HlslExpression() {hlslExpression=Particles.MeshOrigin.x}
    Index 2 Param 2=HlslExpression() {hlslExpression=Particles.MeshOrigin.y}
    Index 2 Param 3=HlslExpression() {hlslExpression=Particles.MeshOrigin.z}
    Index 2 Param 4=$Particles.PendulumMaskFalloff
    Index 3 Param 1=&MakeFloatFromVector(Vector=Linked() {linkedVariable {name=Particles.Position, type=NiagaraPosition}}, Channel=::NewEnumerator0)
    Index 3 Param 2=&MakeFloatFromVector(Vector=Linked() {linkedVariable {name=Particles.Position, type=NiagaraPosition}}, Channel=::NewEnumerator1)
    Index 3 Param 3=&MakeFloatFromVector(Vector=Linked() {linkedVariable {name=Particles.Position, type=NiagaraPosition}}, Channel=::NewEnumerator2)
    Index 3 Param 4=$Particles.PendulumRigidity
    Module.Param0WriteEnabled2=true
    Module.Param0WriteEnabled3=true
    Module.Param1WriteEnabled2=true
    Module.Param1WriteEnabled3=true
    Module.Param2WriteEnabled2=true
    Module.Param2WriteEnabled3=true
    Module.Param3WriteEnabled2=true
    Module.Param3WriteEnabled3=true
    Write Parameter Index 0=false
    Write Parameter Index 2=true
    Write Parameter Index 3=true
   )
   &ScaleMeshSize(Scale Factor=&MakeVector(
     X=&FloatFromCurve(FloatCurve=DataInterface() {propertyValues {Curve {keys [<at 0 0.2>,<at 0.7 1>]}}})
     Y=&FloatFromCurve(FloatCurve=DataInterface() {propertyValues {Curve {keys [<at 0 0.2>,<at 0.7 1>]}}})
     Z=&FloatFromCurve(FloatCurve=DataInterface() {propertyValues {Curve {keys [
          <at 0 0.25 RCIM_Cubic>
          <at 0.7 0.8055 RCIM_Cubic>
          {arriveTangent=-1.5491, interpMode=RCIM_Cubic, leaveTangent=-1.5491, tangentMode=RCTM_User, time=1, value=0.25}
         ]}}})
    ))
   &ScaleColor(ScaleA=true, ScaleRGB=true) {enabled=false}
  ]
  rendererClasses [&NiagaraMeshRendererProperties]
  renderers [{rendererClass=&NiagaraMeshRendererProperties, rendererIndex=0}]
 }
}
summary {
 systemName=NS_WeepingWoods_Winter_Backpack
 emitters [
  {emitterName=Drool, rendererClasses [&NiagaraMeshRendererProperties]}
  {emitterName=Bees, rendererClasses [&NiagaraMeshRendererProperties]}
  {emitterName=Bees_fly_away, rendererClasses [&NiagaraMeshRendererProperties]}
 ]
 userVariables [
  User.bIsFrontEnd {type=NiagaraBool}
  User.bIsFrontEndPreview {type=NiagaraBool}
  User.IsSkydiving {type=NiagaraInt32}
  User.SkeletalMesh {
   type=NiagaraDataInterfaceSkeletalMesh
   defaultValue {
    struct=NiagaraExt_VariableValue_DataInterface
    value {dataInterface=&self:NiagaraDataInterfaceSkeletalMesh_0}
   }
  }
 ]
}
